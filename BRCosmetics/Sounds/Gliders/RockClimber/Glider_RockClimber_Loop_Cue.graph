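# Sound cue for the looping audio of the RockClimber cosmetic glider.
# A mixer layers two looping wave players: the glider's own loop at higher volume and a quieter shared Athena vertical skydive wind loop.
# The cue plays pitched up at reduced overall volume, using the glider thrust attenuation, the cosmetic glider loop audio properties sheet and the Fort_Gliders_Land sound class.
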SoundCue /BRCosmetics/Sounds/Gliders/RockClimber/Glider_RockClimber_Loop_Cue
root=Mixer_0
nodes [
 {id=Mixer_0, children [WavePlayer_1,WavePlayer_0], properties {inputVolume [1.1,0.45]}}
 {id=WavePlayer_0, properties {bLooping=true, soundWaveAssetPtr=/Game/Athena/Sounds/Wind_Falling/AMB_Wind_Skydive_Vertical_Loop_01}}
 {id=WavePlayer_1, properties {bLooping=true, soundWaveAssetPtr=&/Glider_RockClimber_Loop03}}
]
properties {
 attenuationSettings=&Fort_Player_Glider_Thrust_Athena_Att
 audioPropertiesSheet=&APS_Cosmetic_Glider_Loop
 maxDistance=12500
 pitchMultiplier=1.2
 soundClassObject=&Fort_Gliders_Land
 volumeMultiplier=0.4
}
