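# Sound cue for the opening sound of the RockClimber cosmetic glider.
# A fixed 0.27-second delay node feeds a single wave player for the Glider_RockClimber_Open wave, with no randomization.
# Spatialization and mix come from shared glider assets: open/close attenuation and concurrency, the cosmetic glider-open property sheet, and the Fort_Gliders_Land sound class.
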SoundCue /BRCosmetics/Sounds/Gliders/RockClimber/Glider_RockClimber_Open_Cue
root=Delay_0
nodes [
 {id=Delay_0, children [WavePlayer_1], properties {delayMax=0.27, delayMin=0.27}}
 {id=WavePlayer_1, properties {soundWaveAssetPtr=&/Glider_RockClimber_Open}}
]
properties {
 attenuationSettings=/Game/Sounds/Attenuation/Fort_Player_Glider_OpenClose_Athena_Att
 audioPropertiesSheet=/BRCosmetics/Sounds/Settings/PropertySheets/Glider/APS_Cosmetic_Glider_Open
 maxDistance=10000
 soundClassObject=&Fort_Gliders_Land
 volumeMultiplier=1
 concurrencySet [&Athena_GliderOpenClose_Conc]
}
